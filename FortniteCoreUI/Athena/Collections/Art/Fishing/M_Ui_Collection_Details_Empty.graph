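# Translucent UI material for the Athena fishing collection, apparently drawing a placeholder icon for an empty details view.
# An ItemIcon texture parameter is center-scaled by IconScale and mirrored horizontally, with its alpha clamped to a transparent border and multiplied by Opacity.
# A separate fish texture is thresholded through SmoothStep using edge and softness parameters, then reduced by Opacity and saturated.
# A dark blue MainColor parameter is also defined.
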Material /FortniteCoreUI/Athena/Collections/Art/Fishing/M_Ui_Collection_Details_Empty
graph {
 properties {
  bUsedWithStaticMesh=false
  blendMode=BLEND_Translucent
  materialDomain=MD_UI
  pixelDepthOffsetMode=PDOM_Legacy
  usedShadingModels=Default Lit
 }
 nodes [
  n1 = ScalarParameter() {defaultValue=0.5, parameterName=IconScale}
  n2 = MaterialFunctionCall(Texture Scale=AppendVector(A=Multiply(A=@n1) {constB=-1}, B=@n1)) {materialFunction=&ScaleUVsByCenter}
  n3 = TextureObjectParameter() {
   group=Item
   parameterName=ItemIcon
   samplerSource=SSM_Clamp_WorldGroupSettings
   texture=/Game/UI/Foundation/Textures/Icons/Athena/T-T-Icon-BR-FlopperGadget
  }
  n4 = ScalarParameter() {defaultValue=0.5, parameterName=Opacity, sliderMax=1}
  Multiply(
   A=MaterialFunctionCall!(
    UVs=@n2.UVs
    Alpha=TextureSample(UVs=@n2.UVs, Tex=@n3) {samplerSource=SSM_Clamp_WorldGroupSettings, texture=&T-T-Icon-BR-FlopperGadget}.A
   ) {materialFunction=/Game/UI/Foundation/Items/ClampAlphaToTransparentBorder}.Result
   B=@n4
  )
  n5 = TextureSample(UVs=MaterialFunctionCall(Texture Scale=ScalarParameter() {defaultValue=0.5, parameterName=IconScale}) {materialFunction=&ScaleUVsByCenter}.UVs) {samplerSource=SSM_Clamp_WorldGroupSettings, texture=/Game/Athena/Collections/Art/Fishing/T_Ui_collection_FIsh128px}
  n6 = ScalarParameter() {defaultValue=0.2, parameterName=IconEdge}
  n7 = Add(A=@n6, B=ScalarParameter() {defaultValue=0.01, parameterName=IconEdgeSoftness})
  Saturate(None=Subtract(
    A=MaterialFunctionCall(Alpha=@n5.RGB, Min=@n6, Max=@n7) {materialFunction=&SmoothStep}.Result
    B=@n4
   ))
  VectorParameter() {parameterName=MainColor, defaultValue=<rgba 0 0.004 0.0723 1>}
 ]
}
